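; Player-built level 1 stone solid wall Blueprint with a random chance to appear as a gnome variant.
; The Construction Script, on authority only, rolls a random integer up to GnomeWallMax against GnomeWallChance to set the Gnomed flag.
; On BeginPlay, a gnomed wall swaps its static mesh to GnomeWall and adds an animating material mapping pairing the static gnome brick material with GnomeMaterial.
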(event EventBeginPlay
  ;[0 0 0] Make Gnome Wall
  (if (Variables|Gnome|GetGnomed)
    (Variables|Mesh|SetStaticMesh (Variables|Gnome|GetGnomeWall))
    (Utilities|Array|Add (Variables|PlayerBuilt|GetAnimatingMaterialMappings) (Utilities|Struct|MakeAnimatingMaterialPair "/Game/Packages/PBW/Brick/Materials/Brick_lvl1_Static_INST_Gnome.Brick_lvl1_Static_INST_Gnome" (Variables|Gnome|GetGnomeMaterial)))))

(fn ConstructionScript
  ;[0 0 0] RNG Gnomed
  (Networking|SwitchHasAuthority
    (:Authority
      (Variables|Gnome|SetGnomed (< (Math|Random|RandomIntegerinRange 0 (Variables|Gnome|GetGnomeWallMax)) (Variables|Gnome|GetGnomeWallChance))))))
